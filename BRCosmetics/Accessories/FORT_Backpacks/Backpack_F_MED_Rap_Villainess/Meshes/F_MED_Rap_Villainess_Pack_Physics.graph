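# Physics asset for the Rap Villainess backpack cosmetic mesh, defining ragdoll-style bodies for its dangling accessories.
# Box bodies cover the grenade pin, gun, two money pieces, zipper and root pack bone.
# Each dyn_ bone is constrained directly to the pack in a star layout, with tight twist limits and per-piece swing ranges, narrowest for the gun and widest for the second money piece.
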PhysicsAsset /BRCosmetics/Accessories/FORT_Backpacks/Backpack_F_MED_Rap_Villainess/Meshes/F_MED_Rap_Villainess_Pack_Physics
bodies {
 dyn_grenade_pin [{shapeType=Box, extentX=2.3264, extentY=2.4698, extentZ=0.7809, center=<1.2216 0 0>}]
 dyn_gun [{shapeType=Box, extentX=7.6727, extentY=10.846, extentZ=2.1267, center=<4.049 0 0>}]
 dyn_money_01 [{shapeType=Box, extentX=3.4848, extentY=4.9931, extentZ=0.9267, center=<1.7903 0 0>}]
 dyn_money_02 [{shapeType=Box, extentX=4.1135, extentY=4.1611, extentZ=1.1284, center=<2.1677 0 0>}]
 dyn_zipper [{shapeType=Box, extentX=3.4156, extentY=1.9668, extentZ=1.01, center=<1.707 0 0>}]
 pack [{shapeType=Box, extentX=1.9527, extentY=1.9527, extentZ=1.9527}]
}
constraints [
 {bone1Name=dyn_gun, bone2Name=pack, swing1LimitDegrees=3, swing2LimitDegrees=10, twistLimitDegrees=3}
 {bone1Name=dyn_zipper, bone2Name=pack, swing1LimitDegrees=10, swing2LimitDegrees=20, twistLimitDegrees=2}
 {bone1Name=dyn_money_01, bone2Name=pack, swing1LimitDegrees=10, swing2LimitDegrees=30, twistLimitDegrees=2}
 {bone1Name=dyn_money_02, bone2Name=pack, swing1LimitDegrees=25, swing2LimitDegrees=50, twistLimitDegrees=2}
 {bone1Name=dyn_grenade_pin, bone2Name=pack, swing1LimitDegrees=10, swing2LimitDegrees=40, twistLimitDegrees=2}
]
