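# Animation Blueprint for the Rap Villainess backpack cosmetic, adding physics-driven secondary motion to the pack's skeleton.
# A single RigidBody node simulates the input pose with the pack's physics asset in root-bone space, using overridden gravity, a constant upward external force and scaled, clamped component motion.
# The update event sets a Velocity vector from a jumping flag, but the graph never reads it, as the node's force is a fixed value.
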AnimBlueprint /BRCosmetics/Accessories/FORT_Backpacks/Backpack_F_MED_Rap_Villainess/Meshes/F_MED_Rap_Villainess_Pack_AnimBp
target_skeleton=&/F_MED_Rap_Villainess_Pack_Skeleton
graphs [
 AnimGraph {nodes [
   n1 = RigidBody(ComponentPose=Local To Component(LocalPose=Input Pose()), ExternalForce="0,0,1000") {
    bClampLinearTranslationLimitToRefPose=true
    bOverrideWorldGravity=true
    overridePhysicsAsset=&/F_MED_Rap_Villainess_Pack_Physics
    simulationSpace=BaseBoneSpace
    alphaBoolBlend {blendOption=Linear}
    alphaScaleBiasClamp {bInterpResult=true, interpSpeedDecreasing=4, interpSpeedIncreasing=6}
    baseBoneRef {boneName=root}
    componentAppliedLinearAccClamp=<6000 6000 6000>
    componentLinearAccScale=<0.25 0.25 0.25>
    componentLinearVelScale=<0.25 0.25 0.25>
    overrideWorldGravity {z=-850}
   }
   Output Pose(Result=Component To Local(ComponentPose=@n1)) {layerGroup=None}
  ]}
 EventGraph {nodes [
   Event BlueprintUpdateAnimation()
   > Set Velocity(Velocity:Vector=SelectVector(A="0, 0,1250", B="0, 0, 0", bPickA=Get bIsJumping())) -> (Output_Get:Vector)
   TryGetPawnOwner()
  ]}
]
variables [Velocity {type=Vector}]
